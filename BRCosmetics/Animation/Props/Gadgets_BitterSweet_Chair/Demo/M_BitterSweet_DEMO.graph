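# M_BitterSweet_DEMO is a Default Lit demo material for the BitterSweet chair gadget prop.
# It multiplies the RGB of an ambient occlusion texture by a constant to darken it, and separately samples the BitterSweet normal map.
# The graph has no parameters, and it does not show which material inputs these two branches feed.
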Material /BRCosmetics/Animation/Props/Gadgets_BitterSweet_Chair/Demo/M_BitterSweet_DEMO
graph {
 properties {bAutomaticallySetUsageInEditor=false, usedShadingModels=Default Lit}
 nodes [
  Multiply(A=TextureSample() {texture=&/BitterSweet_AO}.RGB, B=Constant() {r=0.3})
  TextureSample() {samplerType=SAMPLERTYPE_Normal, texture=&/BitterSweet_N}
 ]
}
